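# Physics asset for the collar and chest zipper of the female medium TapDance outfit, for secondary motion on those dynamic bones.
# Capsule bodies on five collar bones and the zipper are each constrained to a matching MSTR bone.
# Collar pieces get tight swing and twist limits, while the zipper swings more widely with little twist.
# Head and neck capsules, linked by free angular constraints, likely act as colliders for the collar.
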PhysicsAsset /BRCosmetics/Characters/Player/Female/Medium/Bodies/F_MED_TapDance/Meshes/F_MED_TapDance_Dyn_collar
bodies {
 dyn_collar_bk_l [{shapeType=Capsule, length=3, radius=5, shapeName=dyn_collar_bk_l_capsule, center=<2 0 4.7112>, rotation=<rot 0 0 90>}]
 dyn_collar_bk_l_MSTR [{shapeType=Capsule, length=0.505, radius=0.505}]
 dyn_collar_bk_r [{shapeType=Capsule, length=3, radius=5, shapeName=dyn_collar_bk_r_capsule, center=<2 0 4.7112>, rotation=<rot 0 0 90>}]
 dyn_collar_bk_r_MSTR [{shapeType=Capsule, length=0.505, radius=0.505}]
 dyn_collar_l [{shapeType=Capsule, length=3, radius=5, shapeName=dyn_collar_l_capsule, center=<2 0 4.0372>, rotation=<rot 0 0 90>}]
 dyn_collar_l_MSTR [{shapeType=Capsule, length=0.505, radius=0.505}]
 dyn_collar_mid [{shapeType=Capsule, length=3, radius=5, shapeName=dyn_collar_mid_capsule, center=<2 0 4.4808>, rotation=<rot 0 0 90>}]
 dyn_collar_mid_MSTR [{shapeType=Capsule, length=0.505, radius=0.505}]
 dyn_collar_r [{shapeType=Capsule, length=3, radius=5, shapeName=dyn_collar_r_capsule, center=<2 0 4.0372>, rotation=<rot 0 0 90>}]
 dyn_collar_r_MSTR [{shapeType=Capsule, length=0.505, radius=0.505}]
 dyn_zipper_chest [{shapeType=Capsule, length=1.385, radius=0.5, center=<1.0138 0 0>, rotation=<rot 90 0 0>}]
 dyn_zipper_chest_MSTR [{shapeType=Capsule, length=0.505, radius=0.505}]
 head [{shapeType=Capsule, length=2.8125, radius=4.8125, rotation=<rot 90 0 0>}]
 neck_01 [{shapeType=Capsule, length=9.2812, radius=3.5, rotation=<rot 90 0 0>}]
 neck_02 [{shapeType=Capsule, length=9.2812, radius=3.5, rotation=<rot 90 0 0>}]
}
constraints [
 {bone1Name=dyn_zipper_chest, bone2Name=dyn_zipper_chest_MSTR, swing1LimitDegrees=45, swing2LimitDegrees=25, twistLimitDegrees=3}
 {bone1Name=dyn_collar_l, bone2Name=dyn_collar_l_MSTR, swing1LimitDegrees=5, swing2LimitDegrees=5, twistLimitDegrees=5}
 {bone1Name=dyn_collar_mid, bone2Name=dyn_collar_mid_MSTR, swing1LimitDegrees=5, swing2LimitDegrees=5, twistLimitDegrees=5}
 {bone1Name=dyn_collar_bk_l, bone2Name=dyn_collar_bk_l_MSTR, swing1LimitDegrees=5, swing2LimitDegrees=5, twistLimitDegrees=5}
 {bone1Name=dyn_collar_bk_r, bone2Name=dyn_collar_bk_r_MSTR, swing1LimitDegrees=5, swing2LimitDegrees=5, twistLimitDegrees=5}
 {bone1Name=dyn_collar_r, bone2Name=dyn_collar_r_MSTR, swing1LimitDegrees=5, swing2LimitDegrees=5, twistLimitDegrees=5}
 {
  bone1Name=head
  bone2Name=neck_02
  swing1LimitDegrees=45
  swing1Motion=Free
  swing2LimitDegrees=45
  swing2Motion=Free
  twistLimitDegrees=45
  twistMotion=Free
 }
 {
  bone1Name=neck_02
  bone2Name=neck_01
  swing1LimitDegrees=45
  swing1Motion=Free
  swing2LimitDegrees=45
  swing2Motion=Free
  twistLimitDegrees=45
  twistMotion=Free
 }
]
