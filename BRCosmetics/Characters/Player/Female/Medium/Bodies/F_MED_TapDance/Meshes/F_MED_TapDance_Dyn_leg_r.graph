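# Physics asset for the right-leg dynamic holster on the female medium TapDance character body, giving it simulated secondary motion.
# Two capsule bodies cover the holster bone and its tiny master anchor bone.
# A single constraint joins them with tight angular limits, allowing slight swing on one axis, none on the other, and a small twist.
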PhysicsAsset /BRCosmetics/Characters/Player/Female/Medium/Bodies/F_MED_TapDance/Meshes/F_MED_TapDance_Dyn_leg_r
bodies {
 dyn_holster_r [{shapeType=Capsule, length=8.4475, radius=3.4375, center=<1.0138 0 -1.9453>, rotation=<rot 90 0 0>}]
 dyn_holster_r_MSTR [{shapeType=Capsule, length=0.505, radius=0.505}]
}
constraints [{bone1Name=dyn_holster_r, bone2Name=dyn_holster_r_MSTR, swing1LimitDegrees=2, swing2LimitDegrees=0, twistLimitDegrees=10}]
