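# Post-process style secondary-motion AnimBP for the TapDance female medium body cosmetic, copying its pose from the attached parent mesh.
# A chest zipper bone constraint feeds a world-space rotation of dyn_root_simspace, then RigidBody collar simulation and AnimDynamics holsters.
# Each update blends that rotation from root socket to yaw-offset actor rotation by clamped pawn speed.
# Leg RigidBody nodes and debug draws are left disconnected.
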AnimBlueprint /BRCosmetics/Characters/Player/Female/Medium/Bodies/F_MED_TapDance/Meshes/F_MED_TapDance_AnimBP
target_skeleton=&/F_MED_TapDance_Skeleton
graphs [
 AnimGraph {nodes [
   n1 = RigidBody!(ComponentPose, Alpha, OverrideWorldGravity:Vector="0,0,-980", ExternalForce) -> (Pose) {
    bClampLinearTranslationLimitToRefPose=true
    bOverrideWorldGravity=true
    overridePhysicsAsset=&/F_MED_TapDance_Dyn_leg_l
    simulationSpace=BaseBoneSpace
    alphaBoolBlend {blendOption=Linear}
    baseBoneRef {boneName=thigh_l}
    componentAppliedLinearAccClamp=<3000 3000 3000>
    componentLinearAccScale {z=0.5}
    componentLinearVelScale {z=0.5}
    overrideWorldGravity {z=-980}
    simSpaceSettings {maxAngularAcceleration=5, maxAngularVelocity=5, maxLinearAcceleration=100, maxLinearVelocity=100, worldAlpha=1, externalLinearDragV=<5 5 3>}
   }
   RigidBody(ComponentPose=@n1, OverrideWorldGravity="0,0,-980") {
    bClampLinearTranslationLimitToRefPose=true
    bOverrideWorldGravity=true
    overridePhysicsAsset=&/F_MED_TapDance_Dyn_leg_r
    simulationSpace=BaseBoneSpace
    alphaBoolBlend {blendOption=Linear}
    baseBoneRef {boneName=thigh_r}
    componentAppliedLinearAccClamp=<3000 3000 3000>
    componentLinearAccScale {z=0.5}
    componentLinearVelScale {z=0.5}
    overrideWorldGravity {z=-980}
    simSpaceSettings {maxAngularAcceleration=5, maxAngularVelocity=5, maxLinearAcceleration=100, maxLinearVelocity=100, worldAlpha=1, externalLinearDragV=<5 5 3>}
   }
   n2 = "Constraint - dyn_zipper_chest_MSTR (PP)"!(
    ComponentPose=Local To Component(LocalPose=Copy Pose From Mesh() {bUseAttachedParent=true})
    Alpha=1
    ConstraintWeights_0=0.5166
    ConstraintWeights_1:float=0.4834
   ) -> (Pose) {
    alphaBoolBlend {blendOption=Linear}
    boneToModify {boneName=dyn_zipper_chest_MSTR}
    constraintSetup [{targetBone {boneName=spine_04}}, {targetBone {boneName=spine_05}}]
    constraintWeights [1,1]
   }
   n3 = "Transform (Modify) Bone - Bone: dyn_root_simspace"(ComponentPose=@n2, Alpha=1, Translation="0,0,0", Rotation=Get root_space_rotation() -> (root_space_rotation:Rotator), Scale="1,1,1") {
    rotationMode=BMM_Replace
    rotationSpace=BCS_WorldSpace
    translationSpace=BCS_BoneSpace
    alphaBoolBlend {blendOption=Linear}
    alphaScaleBiasClamp {bInterpResult=true, interpSpeedDecreasing=2, interpSpeedIncreasing=2}
    boneToModify {boneName=dyn_root_simspace}
   }
   n4 = RigidBody(ComponentPose=@n3, OverrideWorldGravity="0,0,-980") {
    bClampLinearTranslationLimitToRefPose=true
    bOverrideWorldGravity=true
    overridePhysicsAsset=&/F_MED_TapDance_Dyn_collar
    simulationSpace=BaseBoneSpace
    alphaBoolBlend {blendOption=Linear}
    baseBoneRef {boneName=spine_01}
    componentAppliedLinearAccClamp=<3000 3000 3000>
    componentLinearAccScale {z=0.5}
    componentLinearVelScale {z=0.5}
    overrideWorldGravity {z=-980}
    simSpaceSettings {maxAngularAcceleration=5, maxAngularVelocity=5, maxLinearAcceleration=100, maxLinearVelocity=100, worldAlpha=1, externalLinearDragV=<5 5 3>}
   }
   n5 = Anim Dynamics - Bone: dyn_holster_r(ComponentPose=@n4, Alpha=1, ExternalForce="0,0,0") {
    angularDampingOverride=0.999
    angularSpringConstant=20000
    bAngularSpring=true
    bOverrideAngularDamping=true
    bUseGravityOverride=true
    gravityScale=0.35
    simulationSpace=RootRelative
    alphaBoolBlend {blendOption=Linear}
    boundBone {boneName=dyn_holster_r}
    componentLinearAccScale=<1 1 1>
    componentLinearVelScale=<1 1 1>
    gravityOverride {z=-980}
    physicsBodyDefinitions [{
      boundBone {boneName=dyn_holster_r}
      boxExtents=<45 45 5>
      constraintSetup {angularLimitsMax=<2 2 3>, angularLimitsMin=<-2 -2 -3>, angularTarget=<0 0 0>}
      localJointOffset=<-2 0 -1>
     }]
   }
   n6 = Anim Dynamics - Bone: dyn_holster_l(ComponentPose=@n5, Alpha=1, ExternalForce="0,0,0") {
    angularDampingOverride=0.999
    angularSpringConstant=20000
    bAngularSpring=true
    bOverrideAngularDamping=true
    bUseGravityOverride=true
    gravityScale=0.35
    simulationSpace=RootRelative
    alphaBoolBlend {blendOption=Linear}
    boundBone {boneName=dyn_holster_l}
    componentLinearAccScale=<1 1 1>
    componentLinearVelScale=<1 1 1>
    gravityOverride {z=-980}
    physicsBodyDefinitions [{
      boundBone {boneName=dyn_holster_l}
      boxExtents=<45 45 5>
      constraintSetup {angularLimitsMax=<2 2 3>, angularLimitsMin=<-2 -2 -3>, angularTarget=<0 0 0>}
      localJointOffset=<-2 0 -1>
     }]
   }
   Output Pose(Result=Component To Local(ComponentPose=@n6)) {layerGroup=None}
  ]}
 EventGraph {nodes [
   Event BlueprintUpdateAnimation()
   > n1 = Sequence!(execute) -> (then_0, then_1, then_2:Exec, then_3:Exec)
   > Reroute Node()
   Reroute Node(InputPin=@n1.then_1)
   n2 = Get Skel_Mesh()
   n3 = GetSocketRotation!(self:Scene Component Object Reference=@n2, InSocketName:Name=root) -> (ReturnValue:Rotator)
   n4 = "Clamp (Float)"(Value=Get PawnSpeed2D(), Min=0, Max=20)
   n5 = Get Actor Rotation!(self:Actor Object Reference=GetOwningActor()) -> (ReturnValue:Rotator) {is=title K2Node_CallFunction}
   n6 = Set root_space_rotation(
    execute=@n1.then_2
    root_space_rotation:Rotator="Lerp (Rotator)"!(
     A:Rotator=@n3
     B:Rotator=Combine Rotators!(A:Rotator=@n5, B:Rotator="0,-90, 0") -> (ReturnValue:Rotator)
     Alpha:float=NormalizeToRange!(Value:double=@n4, RangeMin:double=0, RangeMax:double=20) -> (ReturnValue:double)
     bShortestPath:bool=true
    ) -> (ReturnValue:Rotator)
   ) -> (Output_Get:Rotator)
   Event BlueprintInitializeAnimation()
   > Set Skel_Mesh(Skel_Mesh=GetOwnerSkeletalMeshComponent())
   n7 = vector + vector(A:Vector=GetSocketLocation!(self:Scene Component Object Reference=@n2, InSocketName:Name=root) -> (ReturnValue:Vector), B:Vector="0, 0,165") -> (ReturnValue:Vector) {is=title K2Node_CommutativeAssociativeBinaryOperator}
   DrawDebugCoordinateSystem!(
    execute
    AxisLoc:Vector=@n7
    AxisRot:Rotator=@n3
    Scale:float=100
    Duration:float=0
    Thickness:float=1
    DepthPriority:EDrawDebugSceneDepthPriorityGroup Enum=World
   ) -> (then)
   n8 = GetOwningActor()
   n9 = Get Actor Location(self=@n8)
   n10 = vector * float(A=GetActorForwardVector!(self:Actor Object Reference=@n8) -> (ReturnValue:Vector), B=200)
   DrawDebugLine!(
    execute
    LineStart:Vector=vector + vector(A=@n9, B="0,0,87")
    LineEnd:Vector=vector + vector(A=vector + vector(A=@n9, B=@n10), B="0,0,87")
    LineColor:Linear Color Structure=(R=0.6348,G=0,B=0.6406,A=1)
    Duration:float=0
    Thickness:float=1.5
    DepthPriority:EDrawDebugSceneDepthPriorityGroup Enum=World
   ) -> (then)
   DrawDebugCoordinateSystem(AxisLoc=@n7, AxisRot=@n6, Scale=150, Duration=0, Thickness=2.5, DepthPriority=World)
  ]}
]
variables [
 root_space_rotation {type=Rotator}
 Skel_Mesh {type=SkeletalMeshComponent}
]
