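# Control Rig for the F_MED_TapDance face accessory that applies pose-driven corrective rotations to hair base and dynamic bones, largely as a skydiving fix.
# skydiving_fix uses DeltaRotation to turn clavicle and left upper-arm rotation against reference nulls into clamped weights, smoothed by AlphaInterp into ModifyTransforms.
# Pub_Skydiving gates the right-clavicle and left-upper-arm corrections, while the left-clavicle correction always runs.
# A further skydiving-only offset is left unconnected to execution.
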ControlRigBlueprint /BRCosmetics/Characters/Player/Female/Medium/Bodies/F_MED_TapDance/Meshes/Parts/F_MED_TapDance_FaceAcc_CtrlRig
graphs [
 New Function_ContainedGraph {nodes [
   n1 = RigVMFunctionEntryNode()
   RigVMFunction_MathBoolToInteger = &RigVMFunction_MathBoolToInteger(Value=@n1.onlyTwist)
   RigVMFunction_MathQuaternionSwingTwist = &RigVMFunction_MathQuaternionSwingTwist(
    TwistAxis=@n1.twistAxis
    Input=&RigUnit_GetRelativeTransformForItem(bChildInitial=False, bParentInitial=False, Child=@n1.child, Parent=@n1.parent).RelativeTransform.Rotation
   )
   n2 = RigVMDispatchNode(
    Values=((),())
    Index=@RigVMFunction_MathBoolToInteger.Result
    Values.0=@RigVMFunction_MathQuaternionSwingTwist.Swing
    Values.1=@RigVMFunction_MathQuaternionSwingTwist.Twist
   ) {template="DISPATCH_RigVMDispatch_SelectInt32(in Index,in Values,out Result)"}
   RigVMFunction_MathQuaternionToEuler = &RigVMFunction_MathQuaternionToEuler(Value=@n2.Result, RotationOrder=@n1.Rotation Order)
   > DISPATCH_RigVMDispatch_Print = RigVMDispatchNode(ScreenColor=(R=1,G=1,B=1,A=1), ScreenDuration=1.5, Enabled=@n1.debugRotation, Value=@RigVMFunction_MathQuaternionToEuler.Result) {template="DISPATCH_RigVMDispatch_Print(in Prefix,in Value,in Enabled,in ScreenDuration,in ScreenColor)"}
   Cast = RigVMDispatchNode(Value=@n1.triggerAxis) {template="DISPATCH_RigVMDispatch_CastEnumToInt(in Value,out Result)"}
   DISPATCH_RigVMDispatch_SelectInt32 = RigVMDispatchNode(
    Values=(0,0,0)
    Index=@Cast.Result
    Values.0=@RigVMFunction_MathQuaternionToEuler.Result.X
    Values.1=@RigVMFunction_MathQuaternionToEuler.Result.Y
    Values.2=@RigVMFunction_MathQuaternionToEuler.Result.Z
   ) {template="DISPATCH_RigVMDispatch_SelectInt32(in Index,in Values,out Result)"}
   Remap = &RigVMFunction_MathDoubleRemap(
    bClamp=False
    Value=@DISPATCH_RigVMDispatch_SelectInt32.Result
    TargetMinimum=@n1.outputMinMax.X
    TargetMaximum=@n1.outputMinMax.Y
    SourceMinimum=@n1.rotationMinMax.X
    SourceMaximum=@n1.rotationMinMax.Y
   )
   Absolute = &RigVMFunction_MathDoubleAbs(Value=@Remap.Result)
   Multiply = &RigVMFunction_MathDoubleMul(
    A=&RigVMFunction_MathDoubleClamp(Value=RigVMDispatchNode(True=@Absolute.Result, Condition=@n1.absoluteValue, False=@Remap.Result) {template="DISPATCH_RigVMDispatch_If(in Condition,in True,in False,out Result)"}.Result).Result
    B=@n1.premult
   )
   > RigVMFunctionReturnNode(outputQuat=@n2.Result, outputWeight=@Multiply.Result)
  ]}
 New Function_ContainedGraph {nodes [
   n1 = RigVMFunctionEntryNode()
   RigVMFunction_Sequence = &RigVMFunction_Sequence(ExecuteContext=@n1.ExecuteContext)
   DeltaRotation v4 = RigVMFunctionReferenceNode(
    Rotation Order=ZYX
    child=(Type=Bone,Name="clavicle_r")
    debugRotation=false
    outputMinMax=(X=0,Y=1)
    parent=(Type=Null,Name="clavicle_r_null")
    premult=1
    rotationMinMax=(X=0,Y=-18)
    triggerAxis=Y
    twistAxis=(X=1,Y=0,Z=0)
    ExecuteContext=@RigVMFunction_Sequence.A
   )
   Multiply = &RigVMFunction_MathDoubleMul(
    A=@"DeltaRotation v4".outputWeight
    B=RigVMDispatchNode(False=0, True=1, Condition=RigVMVariableNode() {variable=Pub_Skydiving}.Value) {template="DISPATCH_RigVMDispatch_If(in Condition,in True,in False,out Result)"}.Result
   )
   n2 = &RigVMFunction_AlphaInterp(InterpSpeedDecreasing=5, InterpSpeedIncreasing=5, Scale=1, bClampResult=False, Value=@Multiply.Result)
   > &RigUnit_ModifyTransforms(
    ItemToModify="((Item=(Type=Bone,Name=\"base_hair_01_r\"),Transform=(Rotation=(X=0.133,Y=-0.0155,Z=-0.1314,W=-0.9822),Translation=(X=0,Y=0,Z=0),Scale3D=(X=1,Y=1,Z=1))),(Item=(Type=Bone,Name=\"base_hair_bk_01_r\"),Transform=(Rotation=(X=-0.0562,Y=-0.0034,Z=0.0604,W=0.9966),Tr ...(truncated)"
    WeightMinimum=0
    Weight=@n2.Result
   )
   > DeltaRotation v4_6 = RigVMFunctionReferenceNode(
    Rotation Order=ZYX
    child=(Type=Bone,Name="clavicle_l")
    debugRotation=false
    onlyTwist=false
    outputMinMax=(X=0,Y=1)
    parent=(Type=Null,Name="clavicle_l_null")
    premult=1
    rotationMinMax=(X=0,Y=-10)
    triggerAxis=Y
    twistAxis=(X=1,Y=0,Z=0)
   )
   n3 = &RigVMFunction_AlphaInterp(InterpSpeedDecreasing=5, InterpSpeedIncreasing=5, Scale=1, bClampResult=False, Value=@"DeltaRotation v4_6".outputWeight)
   > &RigUnit_ModifyTransforms(
    ItemToModify=((Item=(Type=Bone,Name="base_hair_bk_01_l"),Transform=(Rotation=(X=-0.0014,Y=0.0829,Z=0.0198,W=-0.9964),Translation=(X=0,Y=0,Z=0),Scale3D=(X=1,Y=1,Z=1))))
    WeightMinimum=0
    Weight=@n3.Result
   )
   > DeltaRotation v4_1 = RigVMFunctionReferenceNode(
    Rotation Order=ZYX
    child=(Type=Bone,Name="upperarm_l")
    debugRotation=false
    onlyTwist=false
    outputMinMax=(X=0,Y=1)
    parent=(Type=Null,Name="upperarm_l_null")
    premult=1
    rotationMinMax=(X=0,Y=-16)
    triggerAxis=Z
    twistAxis=(X=1,Y=0,Z=0)
   )
   n4 = RigVMVariableNode() {variable=Pub_Skydiving}
   n5 = &RigVMFunction_AlphaInterp(
    InterpSpeedDecreasing=5
    InterpSpeedIncreasing=5
    Scale=1
    bClampResult=False
    Value=&RigVMFunction_MathDoubleMul(
     A=@"DeltaRotation v4_1".outputWeight
     B=RigVMDispatchNode(False=0, True=1, Condition=@n4.Value) {template="DISPATCH_RigVMDispatch_If(in Condition,in True,in False,out Result)"}.Result
    ).Result
   )
   > &RigUnit_ModifyTransforms(
    ItemToModify="((Item=(Type=Bone,Name=\"base_hair_bk_01_l\"),Transform=(Rotation=(X=0.0069,Y=-0.1854,Z=-0.0124,W=0.9826),Translation=(X=0,Y=0,Z=0),Scale3D=(X=1,Y=1,Z=1))),(Item=(Type=Bone,Name=\"dyn_hair_bk_03_l\"),Transform=(Rotation=(X=0,Y=0.4868,Z=0,W=0.8735),Translation= ...(truncated)"
    WeightMinimum=0
    Weight=@n5.Result
   )
   RigVMFunctionReturnNode(ExecuteContext=@RigVMFunction_Sequence.B)
   RigVMFunction_MathBoolToFloat = &RigVMFunction_MathBoolToFloat(Value=RigVMVariableNode() {variable=Pub_Skydiving}.Value)
   n6 = &RigVMFunction_AlphaInterp(InterpSpeedDecreasing=5, InterpSpeedIncreasing=5, Scale=1, bClampResult=False, Value=@RigVMFunction_MathBoolToFloat.Result)
   &RigUnit_ModifyTransforms(
    ExecutePin=()
    ItemToModify=((Item=(Type=Bone,Name="base_hair_bk_01_l"),Transform=(Rotation=(X=0.3592,Y=-0.0827,Z=0.2013,W=0.9075),Translation=(X=0,Y=0,Z=0),Scale3D=(X=1,Y=1,Z=1))))
    WeightMinimum=0
    Weight=@n6.Result
   )
  ]}
 RigVMFunctionLibrary {nodes [
   DeltaRotation v4 = RigVMCollapseNode(
    ExecuteContext=()
    Rotation Order=ZYX
    child=(Type=Bone,Name="None")
    outputMinMax=(X=0,Y=1)
    parent=(Type=Bone,Name="None")
    premult=1
    rotationMinMax=(X=0,Y=0)
    twistAxis=(X=1,Y=0,Z=0)
   )
   skydiving_fix = RigVMCollapseNode(ExecuteContext=())
  ]}
 RigVMModel {nodes [
   RigUnit_BeginExecution = &RigUnit_BeginExecution()
   skydiving_fix = RigVMFunctionReferenceNode(ExecuteContext=@RigUnit_BeginExecution.ExecutePin)
  ]}
]
hierarchy [
 root {loc=<0 0 0>, rot=<0 0 0 1>}
 pelvis {parent=root, loc=<0 -0.5301 98.0571>, rot=<0.0127 0.707 -0.0127 -0.707>}
 spine_01 {parent=pelvis, loc=<6.1208 -0.9328 0>, rot=<0 0 0.034 -0.9994>}
 spine_02 {parent=spine_01, loc=<6.5996 -0.2095 0>, rot=<0 0 0.0238 -0.9997>}
 spine_03 {parent=spine_02, loc=<6.4914 -0.1237 0>, rot=<0 0 0.0149 -0.9999>}
 spine_04 {parent=spine_03, loc=<6.9598 0.0794 0>, rot=<0 0 -0.0132 -0.9999>}
 spine_05 {parent=spine_04, loc=<6.9987 0 0>, rot=<0 0 -0.0015 -1>}
 clavicle_l {parent=spine_05, loc=<5.4831 -0.1053 -3.2946>, rot=<0.0465 -0.7289 -0.009 -0.6829>}
 upperarm_l {parent=clavicle_l, loc=<9.8658 -1.2136 -0.4105>, rot=<0.0208 -0.486 -0.0242 -0.8734>}
 clavicle_r {parent=spine_05, loc=<5.4831 -0.1053 3.2946>, rot=<0.7289 0.0465 -0.6829 0.009>}
 upperarm_r {parent=clavicle_r, loc=<-9.8658 1.2136 0.4104>, rot=<0.0208 -0.486 -0.0242 -0.8734>}
 upperarm_r_null {parent=clavicle_r, type=Null, loc=<-9.8658 1.2136 0.4104>, rot=<0.0208 -0.486 -0.0242 -0.8734>}
 neck_01 {parent=spine_05, loc=<11.3138 0.1529 0>, rot=<0 0 -0.2003 -0.9797>}
 neck_02 {parent=neck_01, loc=<4.5627 0 0>, rot=<0 0 0.0504 -0.9987>}
 head {parent=neck_02, loc=<5.3914 0 0>, rot=<0 0 0.115 -0.9934>}
 faceAttach {parent=head, loc=<0.6201 0.9296 0>, rot=<0.0082 -0.7071 -0.0082 -0.7071>, scale=<0.89 0.89 0.89>}
 C_brow_mid {parent=faceAttach, loc=<-0.0028 8.0203 3.8499>, rot=<0 -0.0156 0.0031 -0.9999>}
 L_brow_mid {parent=faceAttach, loc=<3.0646 7.6699 4.8659>, rot=<-0.0001 -0.004 0.2553 -0.9669>}
 L_brow_outer {parent=faceAttach, loc=<5.2138 6.2122 4.6566>, rot=<0.2731 0.3402 0.3324 -0.8361>}
 R_brow_mid {parent=faceAttach, loc=<-3.0642 7.67 4.8666>, rot=<-0.0001 0.0042 -0.2553 -0.9669>}
 base_hair_01_l {parent=head, loc=<9.1278 -2.1505 -5.8242>, rot=<0.0265 -0.9978 0.0011 -0.0608>}
 dyn_hair_01_l {parent=base_hair_01_l, loc=<0 0 0>, rot=<0 0 0 1>}
 dyn_hair_02_l {parent=dyn_hair_01_l, loc=<8.2909 0.1423 0>, rot=<0 0 -0.0446 -0.999>}
 dyn_hair_03_l {parent=dyn_hair_02_l, loc=<7.6817 0 0>, rot=<0 0 -0.2968 -0.9549>}
 dyn_hair_04_l {parent=dyn_hair_03_l, loc=<6.2352 0 0>, rot=<0 0 0 1>}
 dyn_hair_fr_01_l {parent=dyn_hair_01_l, loc=<3.0506 0.9209 1.5556>, rot=<0.0366 -0.0379 -0.0179 -0.9985>}
 dyn_hair_fr_02_l {parent=dyn_hair_fr_01_l, loc=<5.5 0 0>, rot=<0.0015 -0.016 -0.0727 -0.9972>}
 dyn_hair_fr_03_l {parent=dyn_hair_fr_02_l, loc=<6.8889 0 0>, rot=<0 0 -0.226 -0.9741>}
 dyn_hair_fr_04_l {parent=dyn_hair_fr_03_l, loc=<6.203 0 0>, rot=<0 0 0 1>}
 base_hair_bk_01_l {parent=head, loc=<8.9389 -5.3808 -3.5833>, rot=<0.0709 -0.9151 0.3815 -0.1096>}
 dyn_hair_bk_01_l {parent=base_hair_bk_01_l, loc=<0 0 0>, rot=<0 0 0 1>}
 dyn_hair_bk_02_l {parent=dyn_hair_bk_01_l, loc=<5.7433 0 -0.0798>, rot=<0 -0.0209 0 -0.9998>}
 dyn_hair_bk_03_l {parent=dyn_hair_bk_02_l, loc=<7.3578 0 0>, rot=<0 0.0471 0 -0.9989>}
 dyn_hair_bk_04_l {parent=dyn_hair_bk_03_l, loc=<6.8014 0 0>, rot=<0 0 0 1>}
 base_hair_bk_01 {parent=head, loc=<8.9651 -7.2927 0>, rot=<0.1116 -0.6982 0.6982 -0.1116>}
 dyn_hair_bk_01 {parent=base_hair_bk_01, loc=<0 0 0>, rot=<0 0 0 1>}
 dyn_hair_bk_02 {parent=dyn_hair_bk_01, loc=<5.4856 0 0>, rot=<0 -0.099 0 -0.9951>}
 dyn_hair_bk_03 {parent=dyn_hair_bk_02, loc=<7.1779 0 0>, rot=<0 0.0031 0 -1>}
 dyn_hair_bk_04 {parent=dyn_hair_bk_03, loc=<7.538 0 0>, rot=<0 0.0159 0 -0.9999>}
 base_hair_01_r {parent=head, loc=<8.55 -1.0135 5.7883>, rot=<0.0643 -0.0725 0.9941 -0.0487>}
 dyn_hair_01_r {parent=base_hair_01_r, loc=<0 0 0>, rot=<0 0 0 1>}
 dyn_hair_02_r {parent=dyn_hair_01_r, loc=<8.2913 -0.1423 0>, rot=<0 0 0 1>}
 dyn_hair_03_r {parent=dyn_hair_02_r, loc=<7.6813 0 0.0001>, rot=<0 0 0 1>}
 base_hair_bk_01_r {parent=head, loc=<8.9386 -5.3808 3.5833>, rot=<0.1096 -0.3815 0.9151 -0.0709>}
 dyn_hair_bk_01_r {parent=base_hair_bk_01_r, loc=<0 0 0>, rot=<0 0 0 1>}
 dyn_hair_bk_02_r {parent=dyn_hair_bk_01_r, loc=<5.7429 0 -0.0798>, rot=<0 -0.0209 0 -0.9998>}
 dyn_hair_bk_03_r {parent=dyn_hair_bk_02_r, loc=<7.3578 0 0>, rot=<0 0.0471 0 -0.9989>}
 dyn_hair_bk_04_r {parent=dyn_hair_bk_03_r, loc=<6.8019 0 -0.0002>, rot=<0 0 0 1>}
 base_hair_fr_01_r {parent=head, loc=<7.5735 3.9797 5.9941>, rot=<0.0608 -0.0014 0.9976 -0.0317>}
 dyn_hair_fr_01_r {parent=base_hair_fr_01_r, loc=<0 0 0>, rot=<0 0 0 1>}
 dyn_hair_fr_02_r {parent=dyn_hair_fr_01_r, loc=<8.2913 -0.1423 0>, rot=<-0.0004 -0.0382 0.0096 -0.9992>}
 dyn_hair_fr_03_r {parent=dyn_hair_fr_02_r, loc=<7.6813 0 0.0001>, rot=<-0.0046 -0.0334 0.1366 -0.9901>}
 dyn_hair_fr_04_r {parent=dyn_hair_fr_03_r, loc=<6.2351 -0.0001 0>, rot=<0 0 0 -1>}
 clavicle_l_null {parent=spine_05, type=Null, loc=<5.4831 -0.1053 -3.2946>, rot=<0.0465 -0.7289 -0.009 -0.6829>}
 clavicle_r_null {parent=spine_05, type=Null, loc=<5.4831 -0.1053 3.2946>, rot=<0.7289 0.0465 -0.6829 0.009>}
 upperarm_l_null {parent=spine_05, type=Null, loc=<4.5519 -1.8922 -13.0371>, rot=<-0.0415 0.9695 0.017 0.2411>}
 spine_05_null {parent=pelvis, type=Null, loc=<32.9717 -4.2232 0>, rot=<0 0 0.058 -0.9983>}
]
variables [Pub_Skydiving {type=bool, default=False}]
