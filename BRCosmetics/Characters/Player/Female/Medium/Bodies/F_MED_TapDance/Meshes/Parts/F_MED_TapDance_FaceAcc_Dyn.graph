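# Physics asset for the TapDance female medium cosmetic's FaceAcc part, setting up dynamic hair physics.
# Capsule bodies form seven hair chains (center back, plus left and right back, side and front), each rooted at the head with limited swing and twist.
# Extra limited or free-angle constraints cross-link some neighboring strands.
# Head, neck, upper spine, clavicle and upper-arm capsules, joined by free-angle constraints, give the hair a body shape.
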PhysicsAsset /BRCosmetics/Characters/Player/Female/Medium/Bodies/F_MED_TapDance/Meshes/Parts/F_MED_TapDance_FaceAcc_Dyn
bodies {
 clavicle_l [
  {shapeType=Capsule, length=9.875, radius=6.2812, center=<6.2482 -0.659 -0.6805>, rotation=<rot 84.394 0 0>}
  {shapeType=Capsule, length=4.0286, radius=2.8005, center=<10.6695 -0.659 7.7604>, rotation=<rot 84.394 0 0>}
 ]
 clavicle_r [
  {shapeType=Capsule, length=9.875, radius=6.2812, center=<-6.354 0.659 -0.3979>, rotation=<rot 84.394 0 0>}
  {shapeType=Capsule, length=6.4618, radius=2.832, center=<-10.5145 0.659 -9.0983>, rotation=<rot 84.394 0 0>}
 ]
 dyn_hair_01_l [{shapeType=Capsule, length=0, radius=5, shapeName=dyn_hair_01_l_capsule, center=<5.6308 0 6.6>, rotation=<rot -4.921 0.0005 -0.0001>}]
 dyn_hair_01_r [{shapeType=Capsule, length=3.5, radius=2.5625, center=<3.6172 1.9034 1.5828>, rotation=<rot 45.121 103.9265 99.8406>}]
 dyn_hair_02_l [{shapeType=Capsule, length=0, radius=5, shapeName=dyn_hair_02_l_capsule, center=<3.4431 0 6>}]
 dyn_hair_02_r [{shapeType=Capsule, length=3.25, radius=2, center=<1.8718 3.2586 1.9938>, rotation=<rot 45.8979 96.1497 85.9936>}]
 dyn_hair_03_l [{shapeType=Capsule, length=5.1875, radius=0.9062, center=<2.7427 -0.0003 0>, rotation=<rot 90 0 0>}]
 dyn_hair_03_r [{shapeType=Capsule, length=2, radius=2, center=<-0.2863 3.8328 1.4401>, rotation=<rot 45.4349 99.818 84.7674>}]
 dyn_hair_04_l [{shapeType=Capsule, length=5, radius=1.0938, center=<2.6808 -0.0003 -0.4003>, rotation=<rot 81.1954 0 0>}]
 dyn_hair_bk_01 [{shapeType=Capsule, length=0, radius=5, shapeName=dyn_hair_bk_01_capsule, center=<4.9436 0 0.8>}]
 dyn_hair_bk_01_l [{shapeType=Capsule, length=0, radius=4, shapeName=dyn_hair_bk_01_l_capsule, center=<4.0848 0 2>, rotation=<rot 90 0 0>}]
 dyn_hair_bk_01_r [{shapeType=Capsule, length=3.9062, radius=2.1562, center=<3.6172 0 1.5828>, rotation=<rot 90 0 0>}]
 dyn_hair_bk_02 [{shapeType=Capsule, length=0, radius=5, shapeName=dyn_hair_bk_02_capsule, center=<4.9436 0 0.8>}]
 dyn_hair_bk_02_l [{shapeType=Capsule, length=0, radius=4, shapeName=dyn_hair_bk_02_l_capsule, center=<4.0848 0 2>, rotation=<rot 90 0 0>}]
 dyn_hair_bk_02_r [{shapeType=Capsule, length=3.8438, radius=2.2188, center=<3.6172 0 1.5828>, rotation=<rot 90 67.4328 67.2613>}]
 dyn_hair_bk_03 [{shapeType=Capsule, length=0, radius=5, shapeName=dyn_hair_bk_03_capsule, center=<4.9436 0.0052 2.1837>, rotation=<rot 0 0 0.2156>}]
 dyn_hair_bk_03_l [{
   shapeType=Capsule
   length=7.58
   radius=3.3815
   shapeName=dyn_hair_bk_03_l_capsule
   center=<4.0494 -0.6166 3.1338>
   rotation=<rot 56.4857 -93.5753 -90>
  }]
 dyn_hair_bk_03_r [{shapeType=Capsule, length=4.0312, radius=2.0312, center=<0.1204 -0.0104 1.5828>, rotation=<rot 90 -178.2846 -178.4561>}]
 dyn_hair_bk_04 [{shapeType=Capsule, length=0, radius=4, shapeName=dyn_hair_bk_04_capsule, center=<5.5091 0 3.1223>}]
 dyn_hair_bk_04_l [{
   shapeType=Capsule
   length=11.9251
   radius=3.1182
   shapeName=dyn_hair_bk_04_l_capsule
   center=<6.0326 -0.2225 3.2472>
   rotation=<rot 64.3652 -89.9999 -89.9999>
  }]
 dyn_hair_bk_04_r [{shapeType=Capsule, length=11.75, radius=3.5, center=<-0.4 0 0>, rotation=<rot -90 0 -0.0002>}]
 dyn_hair_fr_01_l [{shapeType=Capsule, length=5.8438, radius=0.5, center=<2.7427 -0.0003 0>, rotation=<rot 90 0 0>}]
 dyn_hair_fr_01_r [{shapeType=Capsule, length=4.25, radius=1.8125, center=<3.6172 0.0001 0.8967>, rotation=<rot 90 0 0>}]
 dyn_hair_fr_02_l [{shapeType=Capsule, length=5.8438, radius=0.5, center=<2.7427 -0.0003 0>, rotation=<rot 90 0 0>}]
 dyn_hair_fr_02_r [{shapeType=Capsule, length=4.8125, radius=1.25, center=<3.6172 0 0>, rotation=<rot 90 0 0>}]
 dyn_hair_fr_03_l [{shapeType=Capsule, length=5.8438, radius=1, center=<2.7427 -0.0003 0>, rotation=<rot 90 155.3734 155.3976>}]
 dyn_hair_fr_03_r [{shapeType=Capsule, length=4.8125, radius=1.25, center=<3.6172 0 0>, rotation=<rot 90 0 0>}]
 dyn_hair_fr_04_l [{shapeType=Capsule, length=5.8438, radius=1, center=<2.7427 -0.0003 0>, rotation=<rot 90 60.9872 60.7877>}]
 dyn_hair_fr_04_r [{shapeType=Capsule, length=3.125, radius=1.1875, center=<3.3313 -1.0068 0>, rotation=<rot 90 -14.0363 6.0078>}]
 head [
  {shapeType=Capsule, length=4.375, radius=5.5, center=<2.6046 -1.1977 0.0001>, rotation=<rot 90 0 0>}
  {shapeType=Capsule, length=5.5938, radius=5.3125, center=<1.511 2.7023 0>, rotation=<rot 90 0 0>}
 ]
 neck_01 [{shapeType=Capsule, length=9.9062, radius=5.8125, center=<-1.0594 -0.3599 0>, rotation=<rot 90 16.6993 28.498>}]
 spine_03 [{shapeType=Capsule, length=7.0938, radius=8.7188, center=<7.3489 5.2978 0.0002>}]
 spine_04 [{shapeType=Capsule, length=10.7812, radius=9.4375, center=<5.2863 -1.2885 0>}]
 spine_05 [{shapeType=Capsule, length=6.4375, radius=6.7188, center=<7.3661 -0.778 0.1064>}]
 upperarm_l [
  {shapeType=Capsule, length=2.875, radius=5.2812, center=<1.9198 -0.17 0.4941>, rotation=<rot 82.0505 -0.856 0.7337>}
  {shapeType=Capsule, length=26.0938, radius=3.9062, center=<18.016 0.5101 -0.0001>, rotation=<rot 90 0 0>}
 ]
 upperarm_r [
  {shapeType=Capsule, length=2.875, radius=5.2812, center=<-1.9198 0.17 -0.4941>, rotation=<rot 82.0505 -0.856 0.7337>}
  {shapeType=Capsule, length=26.0938, radius=3.9062, center=<-18.016 0.5101 0.0001>, rotation=<rot 90 0 0>}
 ]
}
constraints [
 {bone1Name=dyn_hair_bk_01, bone2Name=head, swing1LimitDegrees=5, swing2LimitDegrees=30, twistLimitDegrees=5}
 {bone1Name=dyn_hair_bk_02, bone2Name=dyn_hair_bk_01, swing1LimitDegrees=5, swing2LimitDegrees=30, twistLimitDegrees=5}
 {bone1Name=dyn_hair_bk_03, bone2Name=dyn_hair_bk_02, swing1LimitDegrees=5, swing2LimitDegrees=30, twistLimitDegrees=5}
 {bone1Name=dyn_hair_bk_04, bone2Name=dyn_hair_bk_03, swing1LimitDegrees=5, swing2LimitDegrees=30, twistLimitDegrees=5}
 {bone1Name=dyn_hair_bk_01_l, bone2Name=head, swing1LimitDegrees=5, swing2LimitDegrees=10, twistLimitDegrees=5}
 {bone1Name=dyn_hair_bk_02_l, bone2Name=dyn_hair_bk_01_l, swing1LimitDegrees=5, swing2LimitDegrees=10, twistLimitDegrees=5}
 {bone1Name=dyn_hair_bk_03_l, bone2Name=dyn_hair_bk_02_l, swing1LimitDegrees=5, swing2LimitDegrees=30, twistLimitDegrees=5}
 {bone1Name=dyn_hair_bk_04_l, bone2Name=dyn_hair_bk_03_l, swing1LimitDegrees=5, swing2LimitDegrees=30, twistLimitDegrees=5}
 {bone1Name=dyn_hair_01_l, bone2Name=head, swing1LimitDegrees=10, swing2LimitDegrees=10, twistLimitDegrees=5}
 {bone1Name=dyn_hair_02_l, bone2Name=dyn_hair_01_l, swing1LimitDegrees=10, swing2LimitDegrees=10, twistLimitDegrees=5}
 {bone1Name=dyn_hair_03_l, bone2Name=dyn_hair_02_l, swing1LimitDegrees=10, swing2LimitDegrees=30, twistLimitDegrees=5}
 {bone1Name=dyn_hair_04_l, bone2Name=dyn_hair_03_l, swing1LimitDegrees=10, swing2LimitDegrees=30, twistLimitDegrees=5}
 {bone1Name=dyn_hair_fr_01_l, bone2Name=head, swing1LimitDegrees=30, swing2LimitDegrees=30, twistLimitDegrees=5}
 {bone1Name=dyn_hair_fr_02_l, bone2Name=dyn_hair_fr_01_l, swing1LimitDegrees=30, swing2LimitDegrees=30, twistLimitDegrees=5}
 {bone1Name=dyn_hair_fr_03_l, bone2Name=dyn_hair_fr_02_l, swing1LimitDegrees=30, swing2LimitDegrees=30, twistLimitDegrees=5}
 {bone1Name=dyn_hair_fr_04_l, bone2Name=dyn_hair_fr_03_l, swing1LimitDegrees=30, swing2LimitDegrees=30, twistLimitDegrees=5}
 {bone1Name=dyn_hair_bk_01_r, bone2Name=head, swing1LimitDegrees=5, swing2LimitDegrees=30, twistLimitDegrees=5}
 {bone1Name=dyn_hair_bk_02_r, bone2Name=dyn_hair_bk_01_r, swing1LimitDegrees=5, swing2LimitDegrees=30, twistLimitDegrees=5}
 {bone1Name=dyn_hair_bk_03_r, bone2Name=dyn_hair_bk_02_r, swing1LimitDegrees=5, swing2LimitDegrees=30, twistLimitDegrees=5}
 {bone1Name=dyn_hair_01_r, bone2Name=head, swing1LimitDegrees=30, swing2LimitDegrees=30, twistLimitDegrees=5}
 {bone1Name=dyn_hair_02_r, bone2Name=dyn_hair_01_r, swing1LimitDegrees=10, swing2LimitDegrees=30, twistLimitDegrees=5}
 {bone1Name=dyn_hair_03_r, bone2Name=dyn_hair_02_r, swing1LimitDegrees=10, swing2LimitDegrees=30, twistLimitDegrees=5}
 {bone1Name=dyn_hair_fr_01_r, bone2Name=head, swing1LimitDegrees=10, swing2LimitDegrees=15, twistLimitDegrees=5}
 {bone1Name=dyn_hair_fr_02_r, bone2Name=dyn_hair_fr_01_r, swing1LimitDegrees=10, swing2LimitDegrees=15, twistLimitDegrees=5}
 {bone1Name=dyn_hair_fr_03_r, bone2Name=dyn_hair_fr_02_r, swing1LimitDegrees=15, swing2LimitDegrees=10, twistLimitDegrees=10}
 {bone1Name=dyn_hair_fr_04_r, bone2Name=dyn_hair_fr_03_r, swing1LimitDegrees=15, swing2LimitDegrees=10, twistLimitDegrees=10}
 {
  bone1Name=head
  bone2Name=neck_01
  swing1LimitDegrees=45
  swing1Motion=Free
  swing2LimitDegrees=45
  swing2Motion=Free
  twistLimitDegrees=45
  twistMotion=Free
 }
 {
  bone1Name=clavicle_l
  bone2Name=spine_05
  swing1LimitDegrees=45
  swing1Motion=Free
  swing2LimitDegrees=45
  swing2Motion=Free
  twistLimitDegrees=45
  twistMotion=Free
 }
 {
  bone1Name=clavicle_r
  bone2Name=spine_05
  swing1LimitDegrees=45
  swing1Motion=Free
  swing2LimitDegrees=45
  swing2Motion=Free
  twistLimitDegrees=45
  twistMotion=Free
 }
 {
  bone1Name=neck_01
  bone2Name=spine_05
  swing1LimitDegrees=45
  swing1Motion=Free
  swing2LimitDegrees=45
  swing2Motion=Free
  twistLimitDegrees=45
  twistMotion=Free
 }
 {bone1Name=dyn_hair_bk_04_r, bone2Name=dyn_hair_bk_03_r, swing1LimitDegrees=5, swing2LimitDegrees=30, twistLimitDegrees=5}
 {bone1Name=dyn_hair_03_r, bone2Name=dyn_hair_bk_03_r, swing1LimitDegrees=20, swing2LimitDegrees=45, twistLimitDegrees=0}
 {
  bone1Name=upperarm_l
  bone2Name=clavicle_l
  swing1LimitDegrees=45
  swing1Motion=Free
  swing2LimitDegrees=45
  swing2Motion=Free
  twistLimitDegrees=45
  twistMotion=Free
 }
 {
  bone1Name=upperarm_r
  bone2Name=clavicle_r
  swing1LimitDegrees=45
  swing1Motion=Free
  swing2LimitDegrees=45
  swing2Motion=Free
  twistLimitDegrees=45
  twistMotion=Free
 }
 {bone1Name=dyn_hair_bk_03_r, bone2Name=dyn_hair_bk_03, swing1LimitDegrees=20, swing2LimitDegrees=45, twistLimitDegrees=0}
 {
  bone1Name=dyn_hair_bk_03_l
  bone2Name=dyn_hair_bk_03
  swing1LimitDegrees=45
  swing1Motion=Free
  swing2LimitDegrees=45
  swing2Motion=Free
  twistLimitDegrees=45
  twistMotion=Free
 }
 {
  bone1Name=spine_05
  bone2Name=spine_04
  swing1LimitDegrees=45
  swing1Motion=Free
  swing2LimitDegrees=45
  swing2Motion=Free
  twistLimitDegrees=45
  twistMotion=Free
 }
 {
  bone1Name=dyn_hair_fr_03_l
  bone2Name=dyn_hair_03_l
  swing1LimitDegrees=45
  swing1Motion=Free
  swing2LimitDegrees=45
  swing2Motion=Free
  twistLimitDegrees=45
  twistMotion=Free
 }
 {
  bone1Name=dyn_hair_fr_04_l
  bone2Name=dyn_hair_04_l
  swing1LimitDegrees=45
  swing1Motion=Free
  swing2LimitDegrees=45
  swing2Motion=Free
  twistLimitDegrees=45
  twistMotion=Free
 }
 {
  bone1Name=spine_04
  bone2Name=spine_03
  swing1LimitDegrees=45
  swing1Motion=Free
  swing2LimitDegrees=45
  swing2Motion=Free
  twistLimitDegrees=45
  twistMotion=Free
 }
]
