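# Post-copy animation Blueprint for the female TapDance face accessory part, adding secondary motion to its dynamic bones, including back hair.
# It copies the attached parent's pose, layers in Body_Mesh upper-arm branches, then runs RigidBody physics relative to neck_01.
# Two additive hair-bone offsets feed a Control Rig that receives the skydive-floating state.
# The event graph overrides simulation gravity for swim sprinting, parachuting and skydiving, using vertical velocity and pelvis orientation.
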AnimBlueprint /BRCosmetics/Characters/Player/Female/Medium/Bodies/F_MED_TapDance/Meshes/Parts/F_MED_TapDance_FaceAcc_AnimBP
target_skeleton=&/F_MED_TapDance_FaceAcc_Skeleton
graphs [
 AnimGraph {nodes [
   n1 = RigidBody!(
    ComponentPose=Local To Component(LocalPose=Use cached pose 'Start_Cache'())
    Alpha
    OverrideWorldGravity:Vector=Get gravity_override() -> (gravity_override:Vector)
    ExternalForce
   ) -> (Pose) {
    bClampLinearTranslationLimitToRefPose=true
    bOverrideWorldGravity=true
    overridePhysicsAsset=&/F_MED_TapDance_FaceAcc_Dyn
    simulationSpace=BaseBoneSpace
    alphaBoolBlend {blendOption=Linear}
    baseBoneRef {boneName=neck_01}
    componentAppliedLinearAccClamp=<3000 3000 3000>
    componentLinearAccScale=<0.45 0.3 0.3>
    componentLinearVelScale=<0.45 0.3 0.3>
    overrideWorldGravity {z=-980}
   }
   n2 = "Transform (Modify) Bone - Bone: dyn_hair_bk_03_r"(ComponentPose=@n1, Alpha=1, Translation="0,-3,0", Rotation="0.000,0.000,-15", Scale="1,1,1") {
    rotationMode=BMM_Additive
    translationMode=BMM_Additive
    alphaBoolBlend {blendOption=Linear}
    boneToModify {boneName=dyn_hair_bk_03_r}
    rotation {roll=-15}
    translation {y=-3}
   }
   n3 = "Transform (Modify) Bone - Bone: dyn_hair_bk_04_r"(ComponentPose=@n2, Alpha=1, Translation="-0.000,1,-0.000", Rotation="0,0,0", Scale="1,1,1") {
    rotationMode=BMM_Additive
    translationMode=BMM_Additive
    alphaBoolBlend {blendOption=Linear}
    boneToModify {boneName=dyn_hair_bk_04_r}
    translation {y=1}
   }
   n4 = Control Rig!(Source=Component To Local(ComponentPose=@n3), Alpha, Pub_Skydiving:bool=Get bIsSkydiveFloating()) -> (Pose) {
    alphaBoolBlend {blendOption=Linear}
    controlRigAssetReference {blueprintRigClass=&/F_MED_TapDance_FaceAcc_CtrlRig.F_MED_TapDance_FaceAcc_CtrlRig_C}
   }
   Output Pose(Result=@n4) {layerGroup=None}
   n5 = Get Body_Mesh()
   Save cached pose 'Start_Cache'(Pose=Layered blend per bone(BasePose=Copy Pose From Mesh() {bUseAttachedParent=true}, BlendPoses_0=Copy Pose From Mesh(SourceMeshComponent=@n5)) {layerSetup [{branchFilters [{boneName=upperarm_l}, {boneName=upperarm_r}]}]}) {cacheName=Start_Cache}
  ]}
 EventGraph {nodes [
   Event BlueprintInitializeAnimation()
   > n1 = Cast To PlayerPawn_Athena(Object=TryGetPawnOwner())
   > Set Body_Mesh(Body_Mesh=GetSkeletalMeshForPartType(self=@n1))
   > Set Skel_Mesh(Skel_Mesh=GetOwnerSkeletalMeshComponent())
   Event BlueprintUpdateAnimation()
   > Sequence()
   > n2 = Branch(Condition=Get bIsSwimSprinting() -> (bIsSwimSprinting:bool))
   > Set gravity_override(gravity_override="0, 0, 0")
   n3 = Branch(execute=@n2.else, Condition=Get bIsSkyDiving())
   > n4 = Branch(Condition=Get bIsParachuteOpened())
   Set gravity_override(execute=[@n3.else, @n4.then], gravity_override="0, 0,-980")
   n5 = Branch(execute=@n4.else, Condition="float > float"(A=Get PawnVelocityZ(), B=50) {is=title K2Node_CallFunction})
   > Set gravity_override(gravity_override="0,0,200")
   n6 = Get Skel_Mesh()
   n7 = vector * float(
    A=GetForwardVector!(InRot:Rotator=GetSocketRotation!(self:Scene Component Object Reference=@n6, InSocketName:Name=pelvis) -> (ReturnValue:Rotator)) -> (ReturnValue:Vector)
    B=-1000
   )
   Set gravity_override(
    execute=@n5.else
    gravity_override=vector + vector(A:Vector=@n7, B:Vector="0, 0,-200") -> (ReturnValue:Vector) {is=title K2Node_CommutativeAssociativeBinaryOperator}
   )
   TryGetPawnOwner()
   n8 = GetSocketLocation!(self:Scene Component Object Reference=@n6, InSocketName:Name=pelvis) -> (ReturnValue:Vector)
   DrawDebugLine!(
    execute
    LineStart:Vector=@n8
    LineEnd:Vector=vector + vector(A=@n8, B=@n7)
    LineColor:Linear Color Structure=(R=1,G=0,B=0,A=1)
    Duration:float=0
    Thickness:float=3
    DepthPriority:EDrawDebugSceneDepthPriorityGroup Enum=World
   ) -> (then)
  ]}
]
variables [
 Body_Mesh {type=SkeletalMeshComponent}
 gravity_override {type=Vector}
 Skel_Mesh {type=SkeletalMeshComponent}
]
